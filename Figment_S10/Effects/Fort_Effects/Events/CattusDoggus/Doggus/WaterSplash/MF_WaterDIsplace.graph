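# Material function for the Cattus/Doggus event water splash that creates rippling radial displacement around the object's position.
# Distance from the object, scaled by dynamic parameter Radius, is broken up with noise and offset by the TIme parameter to drive a repeating curve-atlas ramp.
# The signed ramp, scaled by Strength and the V coordinate, outputs as Result_Displace and drives a BumpOffset for Result_UV.
# Several leftover nodes, including Time, Sine and a ramp texture, feed no output.
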MaterialFunction /Figment_S10/Effects/Fort_Effects/Events/CattusDoggus/Doggus/WaterSplash/MF_WaterDIsplace
graph {nodes [
  n1 = FunctionInput(Preview=TextureCoordinate()) {bUsePreviewValueAsDefault=true, inputName=UV, inputType=FunctionInput_Vector2}
  Add(A=@n1)
  n2 = TextureSample(UVs=TextureCoordinate() {coordinateIndex=1, uTiling=2, vTiling=2}) {samplerType=SAMPLERTYPE_LinearColor, texture=/Game/Effects/Fort_Effects/Gameplay/SmokeDustExplosions/Textures/T_NoisePack_LinearColor}
  n3 = Subtract(A=@n2.RGB) {constB=0.5}
  n4 = Divide(A=Distance(A=WorldPosition().XYZ, B=ObjectPositionWS())) {constB=-8192}
  n5 = DynamicParameter() {paramNames [Strength,TIme,RefOffset,Radius]}
  n6 = DynamicParameter() {paramNames [Strength,TIme,RefOffset,Radius]}
  n7 = Add(A=Add(A=Multiply(A=@n3) {constB=0.4}, B=Multiply(A=@n4, B=@n5.Radius)), B=@n6.TIme)
  n8 = Multiply(A=ComponentMask(None=@n7) {r=true}) {constB=-1}
  n9 = CurveAtlasRowParameter(CurveTime=Frac(None=@n8)) {
   atlas=&AthenaCurveAtlas01
   curve=/Game/Athena/Prototype/Blueprints/Cattus/FX/Curve_Ramps2
   defaultValue=114
   parameterName=Param_2
  }
  n10 = ComponentMask(None=TextureCoordinate()) {g=true}
  n11 = ComponentMask(None=Multiply(A=ConstantBiasScale(None=@n9.B) {bias=-0.5, scale=2}, B=Multiply(A=@n6.Strength, B=@n10)) {constB=0.2}) {r=true}
  FunctionOutput(None=BumpOffset(Coordinate=@n1, Height=Constant() {r=1}, HeightRatioInput=@n11)) {outputName=Result_UV, sortPriority=1}
  FunctionOutput(None=@n11) {outputName=Result_Displace, sortPriority=2}
  MaterialFunctionCall() {materialFunction=&DebugTimeSine}
  Sine(None=@n7)
  TextureSample(UVs=@n8) {samplerSource=SSM_Wrap_WorldGroupSettings, texture=/Game/Effects/Fort_Effects/Textures/Ramps/T_Ramp_up}
  Time()
 ]}
