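# Masked lit material for mesh particles in the Doggus water splash effect, built on a background ocean shader.
# The ocean layer blends terrain-mask depth colors, panning foam, ground tint and layered normals, with skybox-collection time-of-day lighting and cheaper quality paths.
# A final attribute override replaces emissive with caustics over desaturated ocean color, tinted by particle color and time of day.
# MF_WaterDIsplace distorts UVs and, with particle alpha and a V gradient, drives dithered opacity.
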Material /Figment_S10/Effects/Fort_Effects/Events/CattusDoggus/Doggus/WaterSplash/M_Backround_WatertoLand
graph {
 properties {
  bUseMaterialAttributes=true
  bUsedWithInstancedStaticMeshes=true
  bUsedWithMeshParticles=true
  blendMode=BLEND_Masked
  pixelDepthOffsetMode=PDOM_Legacy
  usedShadingModels=Default Lit
 }
 nodes [
  n1 = CollectionParameter() {
   collection=/Game/Environments/World/Backgrounds/Universal/MasterMaterials/MPC_Skybox_World
   parameterName="Morn_R=Distance_G=DiffuseAmt_B=EmissiveAmt"
  }
  n2 = CollectionParameter() {
   collection=&MPC_Skybox_World
   parameterName="Day_R=Distance_G=DiffuseAmt_B=EmissiveAmt"
  }
  n3 = CollectionParameter() {
   collection=&MPC_Skybox_World
   parameterName="Eve_R=Distance_G=DiffuseAmt_B=EmissiveAmt"
  }
  n4 = CollectionParameter() {
   collection=&MPC_Skybox_World
   parameterName="Night_R=Distance_G=DiffuseAmt_B=EmissiveAmt"
  }
  n5 = MaterialFunctionCall(Float3=MaterialFunctionCall(Color_Morning=@n1, Color_Day=@n2, Color_Evening=@n3, Color_Night=@n4) {materialFunction=&Color_Time_of_Day}.Result) {materialFunction=&BreakOutFloat3Components}
  n6 = ComponentMask(None=WorldPosition() {worldPositionShaderOffset=WPT_ExcludeAllShaderOffsets}.XYZ) {g=true, r=true}
  n7 = SphereMask(A=@n6, B=Constant2Vector(), Radius=@n5.R, Hardness=Constant())
  n8 = MaterialFunctionCall(
   Color_Morning=CollectionParameter() {collection=&MPC_Skybox_World, parameterName=MorningColor}
   Color_Day=CollectionParameter() {collection=&MPC_Skybox_World, parameterName=DayColor}
   Color_Evening=CollectionParameter() {collection=&MPC_Skybox_World, parameterName=EveningColor}
   Color_Night=CollectionParameter() {collection=&MPC_Skybox_World, parameterName=NightColor}
  ) {materialFunction=&Color_Time_of_Day}
  n9 = MaterialFunctionCall(
   Color_Morning=CollectionParameter() {collection=&MPC_Skybox_World, parameterName=MorningLitColor}
   Color_Day=CollectionParameter() {collection=&MPC_Skybox_World, parameterName=DayLitColor}
   Color_Evening=CollectionParameter() {collection=&MPC_Skybox_World, parameterName=EveningLitColor}
   Color_Night=CollectionParameter() {collection=&MPC_Skybox_World, parameterName=NightLitColor}
  ) {materialFunction=&Color_Time_of_Day}
  n10 = CollectionParameter() {
   collection=&MPC_Skybox_World
   parameterName="Morn_R=LitLength_G=LitRot_B=LitOffset"
  }
  n11 = CollectionParameter() {
   collection=&MPC_Skybox_World
   parameterName="Day_R=LitLength_G=LitRot_B=LitOffset"
  }
  n12 = CollectionParameter() {
   collection=&MPC_Skybox_World
   parameterName="Eve_R=LitLength_G=LitRot_B=LitOffset"
  }
  n13 = CollectionParameter() {
   collection=&MPC_Skybox_World
   parameterName="Night_R=LitLength_G=LitRot_B=LitOffset"
  }
  n14 = MaterialFunctionCall(Float3=MaterialFunctionCall(Color_Morning=@n10, Color_Day=@n11, Color_Evening=@n12, Color_Night=@n13) {materialFunction=&Color_Time_of_Day}.Result) {materialFunction=&BreakOutFloat3Components}
  n15 = Subtract(A=Add(A=@n6, B=Constant() {r=4000000}), B=AppendVector(A=@n14.G, B=@n14.G))
  n16 = Constant2Vector()
  n17 = Constant()
  n18 = Multiply(
   A=Multiply(A=@n5.B, B=OneMinus(None=@n7))
   B=LinearInterpolate(
    A=ComponentMask(None=@n8.Result) {b=true, g=true, r=true}
    B=ComponentMask(None=@n9.Result) {b=true, g=true, r=true}
    Alpha=SphereMask(A=@n15, B=@n16, Radius=@n14.R, Hardness=@n17)
   )
  ) {desc=Emissive}
  n19 = VertexInterpolator(VS=@n18)
  n20 = ComponentMask(None=WorldPosition().XYZ) {g=true, r=true}
  n21 = Constant2Vector() {g=-153000, r=-153000}
  n22 = TextureSample(UVs=Divide(A=Subtract(A=@n20, B=@n21), B=Constant() {r=306200})) {
   samplerSource=SSM_Clamp_WorldGroupSettings
   samplerType=SAMPLERTYPE_LinearColor
   texture=/Game/Athena/Prototype/Terrain/Textures/T_Athena_Terrain_Mask
  }
  n23 = Constant() {r=0.2}
  n24 = LinearInterpolate(
   A=VectorParameter() {parameterName=deepwater, defaultValue=<rgba 0.0049 0.0249 0.05 1>}.RGB
   B=VectorParameter() {parameterName=shallowwater, defaultValue=<rgba 0.0469 0.5 0.4635 1>}.RGB
   Alpha=Multiply(A=Clamp(None=@n22.A, Max=@n23) {clampMode=CMODE_ClampMin, maxDefault=0.5}, B=VertexColor().A)
  )
  n25 = ComponentMask(None=Divide(A=WorldPosition().XYZ, B=Constant() {r=6202.6328})) {g=true, r=true}
  n26 = Time()
  n27 = TextureSample(UVs=MaterialFunctionCall!(UV=Panner(Coordinate=@n25, Time=@n26) {bFractionalPart=true, speedX=-0.03, speedY=0.03}) {materialFunction=&/MF_WaterDIsplace}.Result_UV) {texture=/Figment_S10/Packages/Fortnite_UniqueMaterials_1/Textures/Foam}
  n28 = Multiply(A=@n25) {constB=1.5}
  n29 = TextureSample(UVs=MaterialFunctionCall(UV=Panner(Coordinate=@n28, Time=@n26) {bFractionalPart=true, speedX=0.02, speedY=-0.1}) {materialFunction=&MF_WaterDIsplace}.Result_UV) {texture=&Foam}
  n30 = Multiply(A=Constant3Vector() {constant=<rgba 1.5469 2.5102 3 0>}, B=Clamp(None=Subtract(A=@n27.G, B=@n29.G)))
  n31 = Constant() {r=3}
  n32 = Clamp(None=Subtract(A=Multiply(A=@n22.A, B=@n31)) {constB=0.6})
  n33 = QualitySwitch(Default=@n24, Low=@n24, High=Add(A=@n24, B=Multiply(A=@n30, B=@n32)), Medium=@n24)
  n34 = Divide(A=ComponentMask(None=WorldPosition().XYZ) {g=true, r=true}, B=ScalarParameter() {defaultValue=478490.3125, parameterName=TextureSize_1})
  n35 = TextureSample(UVs=MaterialFunctionCall(UV=VertexInterpolator(VS=@n34).PS) {materialFunction=&MF_WaterDIsplace}.Result_UV) {samplerType=SAMPLERTYPE_Masks, texture=/Game/Environments/World/Backgrounds/Transylvania/Textures/T_OceanGround_M}
  n36 = ScalarParameter() {defaultValue=0.2248, parameterName=SandBlend}
  n37 = ScalarParameter() {defaultValue=-0.3143, parameterName=SandBlend_1}
  n38 = LinearInterpolate(
   A=LinearInterpolate(
    A=VectorParameter() {parameterName=Shallow, defaultValue=<rgba 0.0306 0.1031 0.151 1>}.RGB
    B=FeatureLevelSwitch(Default=@n33, ES2=@n24, ES3_1=@n24, SM4_REMOVED=@n33)
    Alpha=@n35.G
   )
   B=VectorParameter() {parameterName=Ground, defaultValue=<rgba 0.1389 0.3107 0.375 1>}.RGB
   Alpha=Clamp(None=LinearInterpolate(A=@n36, B=@n37, Alpha=@n35.G))
  )
  n39 = Constant3Vector() {constant=<rgba 0 0 1 0>}
  n40 = ComponentMask(None=Divide(A=WorldPosition().XYZ, B=Constant() {r=18286.0156})) {g=true, r=true}
  n41 = Multiply(A=Time(), B=Constant() {r=0.6})
  n42 = Panner(Coordinate=Multiply(A=Constant() {r=0.119}, B=@n40), Time=@n41) {bFractionalPart=true, speedX=-0.015, speedY=0.005}
  n43 = Panner(Coordinate=Multiply(A=Constant() {r=0.2557}, B=@n40), Time=@n41) {bFractionalPart=true, speedX=0.01, speedY=-0.005}
  n44 = MaterialFunctionCall(
   BaseNormal=TextureSample(UVs=MaterialFunctionCall(UV=@n42) {materialFunction=&MF_WaterDIsplace}.Result_UV) {samplerType=SAMPLERTYPE_Normal, texture=/Figment_S10/Packages/Fortnite_UniqueMaterials_1/Textures/T_Water2_N}.RGB
   AdditionalNormal=TextureSample(UVs=MaterialFunctionCall(UV=@n43) {materialFunction=&MF_WaterDIsplace}.Result_UV) {samplerType=SAMPLERTYPE_Normal, texture=/Game/Effects/Fort_Effects/Textures/WaterFoam0001_19_N}.RGB
  ) {materialFunction=&BlendAngleCorrectedNormals}
  n45 = MaterialFunctionCall(Fade Length=Constant() {r=185195.2031}, Fade Offset=Constant()) {materialFunction=&CameraDepthFade}
  n46 = LinearInterpolate(A=Constant() {r=1}, B=Constant() {r=0.85}, Alpha=@n45.Result)
  n47 = Time()
  n48 = Panner(Coordinate=@n40, Time=@n47) {bFractionalPart=true, speedX=0.009, speedY=-0.009}
  n49 = Panner(Coordinate=Multiply(A=@n40, B=Constant() {r=1}), Time=@n47) {bFractionalPart=true, speedX=-0.003, speedY=0.006}
  n50 = Add(
   A=TextureSample(UVs=MaterialFunctionCall(UV=@n48) {materialFunction=&MF_WaterDIsplace}.Result_UV) {samplerType=SAMPLERTYPE_Normal, texture=&T_Water2_N}.RGB
   B=TextureSample(UVs=MaterialFunctionCall(UV=@n49) {materialFunction=&MF_WaterDIsplace}.Result_UV) {samplerType=SAMPLERTYPE_Normal, texture=&WaterFoam0001_19_N}.RGB
  )
  n51 = LinearInterpolate(A=Constant() {r=0.2}, B=Constant() {r=1}, Alpha=@n45.Result)
  n52 = Fresnel(ExponentIn=Constant() {r=4}, Normal=VertexNormalWS()) {baseReflectFraction=0}
  n53 = MaterialFunctionCall(
   Normal=Add(
    A=MaterialFunctionCall(Normal=@n44.Result, Flatness=@n46) {materialFunction=&FlattenNormal}.Result
    B=MaterialFunctionCall(Normal=@n50, Flatness=@n51) {materialFunction=&FlattenNormal}.Result
   )
   Flatness=@n52
  ) {materialFunction=&FlattenNormal}
  n54 = MaterialFunctionCall(Normal=@n44.Result, Flatness=LinearInterpolate(A=Constant() {r=0.85}, B=Constant() {r=1}, Alpha=@n52)) {materialFunction=&FlattenNormal}
  n55 = QualitySwitch(Default=@n53.Result, Low=@n54.Result, High=@n53.Result, Medium=@n53.Result)
  n56 = ComponentMask(None=WorldPosition().XYZ) {b=true}
  n57 = Divide(
   A=Add(A=@n56, B=ScalarParameter() {defaultValue=4509.2178, parameterName=Transition_Offset_2})
   B=ScalarParameter() {defaultValue=306871.9375, parameterName=Transition_Distance_1}
  )
  n58 = Add(A=@n56, B=ScalarParameter() {defaultValue=4335.1958, parameterName=TransitionBeach_Offset})
  n59 = ScalarParameter() {defaultValue=779.4005, parameterName=TransitionBeach_Distance}
  n60 = LinearInterpolate(
   A=Constant()
   B=LinearInterpolate(
    A=ScalarParameter() {defaultValue=0.3518, parameterName=FakeFogBrightness}
    B=ScalarParameter() {defaultValue=1, parameterName=FakeFogBrightness_2}
    Alpha=Clamp(None=@n57)
   )
   Alpha=Clamp(None=Divide(A=@n58, B=@n59))
  )
  n61 = MakeMaterialAttributes!(
   BaseColor=LinearInterpolate(A=@n38, B=Multiply(A=@n38, B=@n19.PS), Alpha=@n5.G)
   Metallic
   Specular=ScalarParameter() {defaultValue=0.1, parameterName=S Ocean}
   Roughness=ScalarParameter() {defaultValue=0.2, parameterName=R Ocean_1}
   Anisotropy
   EmissiveColor=VertexInterpolator(VS=Multiply(A=@n60, B=@n18)).PS
   Opacity
   OpacityMask
   Normal=MaterialFunctionCall(
    BaseNormal=Transform(None=@n39) {transformSourceType=TRANSFORMSOURCE_World, transformType=TRANSFORM_Tangent}
    AdditionalNormal=FeatureLevelSwitch(Default=@n55, ES2=@n54.Result, ES3_1=@n54.Result, SM4_REMOVED=@n55)
   ) {materialFunction=&BlendAngleCorrectedNormals}.Result
   Tangent
   WorldPositionOffset
   SubsurfaceColor
   ClearCoat
   ClearCoatRoughness
   AmbientOcclusion
   Refraction
   CustomizedUVs_0
   CustomizedUVs_1
   CustomizedUVs_2
   CustomizedUVs_3
   CustomizedUVs_4
   CustomizedUVs_5
   CustomizedUVs_6
   CustomizedUVs_7
   PixelDepthOffset
   ShadingModel
   Displacement
  )
  n62 = ParticleColor()
  n63 = TextureCoordinate()
  n64 = Multiply(A=OneMinus(None=ComponentMask(None=@n63) {g=true})) {constB=4}
  n65 = MaterialFunctionCall() {materialFunction=&MF_WaterDIsplace}
  n66 = ConstantBiasScale(None=@n65.Result_Displace)
  n67 = Multiply(A=@n64, B=@n66)
  n68 = WorldPosition()
  n69 = Multiply(A=@n66) {constB=4096}
  n70 = ComponentMask(None=Divide(A=Add(A=@n68.XYZ, B=@n69)) {constA=2048, constB=-8192}) {g=true, r=true}
  n71 = LinearInterpolate(Alpha=TextureSample(UVs=Divide(A=@n70) {constB=8}) {texture=/Figment_S10/Effects/Fort_Effects/Textures/Water/T_Caustics_Softer}.R) {constA=0.2, constB=2}
  n72 = LinearInterpolate(Alpha=TextureSample(UVs=@n70) {texture=&T_Caustics_Softer}.R) {constA=0.2, constB=2}
  n73 = DynamicParameter() {paramNames [Strength,TIme,RefOffset,Param4]}
  n74 = GetMaterialAttributes(None=@n61) {attributeGetTypes [#1]}
  n75 = Constant() {r=0.25}
  n76 = Multiply(
   A=Multiply(A=Multiply(A=@n65.Result_Displace, B=@n62.RGB), B=Subtract(A=@n64, B=@n73.RefOffset))
   B=Multiply(A=Desaturation(None=@n74.Base Color, Fraction=@n75), B=@n65.Result_Displace)
  )
  n77 = MaterialFunctionCall() {materialFunction=&Color_Time_of_Day}
  SetMaterialAttributes!(
   MaterialAttributes=@n61
   Opacity Mask=MaterialFunctionCall(Alpha Threshold=Add(A=@n62.A, B=@n67)) {materialFunction=&DitherTemporalAA}.Result
   Pixel Depth Offset
   Emissive Color=Multiply(A=Multiply(A=@n71, B=@n72), B=Multiply(A=@n76, B=@n77.Result))
   Base Color
  ) {attributeSetTypes [
    #2
    #3
    #4
    #1
   ]}
  VertexInterpolator(VS=LinearInterpolate(A=@n7, Alpha=@n5.G))
 ]
}
